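# Sound cue for the equip foley of the PaddleGinger item, routed to the weapon foley sound class.
# A single wave player feeds PaddleGinger_Equip_01 straight to the output at reduced volume, with no randomization or mixing.
# It references the shared small pistol attenuation asset and overrides concurrency to cap simultaneous copies.
# The debug flag is left enabled.
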SoundCue /MotherGameplay/Sounds/PaddleGinger/PaddleGinger_Equip_Cue
root=WavePlayer_1
nodes [{id=WavePlayer_1, properties {soundWaveAssetPtr=&/PaddleGinger_Equip_01}}]
properties {
 attenuationSettings=/Game/Sounds/Attenuation/Fort_Weapon_Pistol_Sml_Att
 bDebug=true
 bOverrideConcurrency=true
 maxDistance=2000
 soundClassObject=&Fort_WeaponFoley
 volumeMultiplier=0.376
 attenuationOverrides {bAttenuateWithLPF=true, lPFFrequencyAtMax=500, lPFRadiusMax=1500, lPFRadiusMin=1000, stereoSpread=800}
 concurrencyOverrides {maxCount=2, retriggerTime=0.1}
}
