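# Sound cue for the PaddleGinger "glitch tell" sound, played at reduced volume.
# A random node picks between two equally weighted waves, and a modulator above it varies pitch.
# It uses a first-person ranged weapon fire sound class, shared small-pistol attenuation and first-person gunfire concurrency.
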SoundCue /MotherGameplay/Sounds/PaddleGinger/PaddleGinger_GlitchTell_Cue
root=Modulator_0
nodes [
 {id=Modulator_0, children [Random_0], properties {pitchMax=1.15}}
 {id=Random_0, children [WavePlayer_1,WavePlayer_2], properties {weights [1,1]}}
 {id=WavePlayer_1, properties {soundWaveAssetPtr=&/PaddleGinger_GlitchTell_01}}
 {id=WavePlayer_2, properties {soundWaveAssetPtr=&/PaddleGinger_GlitchTell_02}}
]
properties {
 attenuationSettings=/Game/Sounds/Attenuation/Fort_Weapon_Pistol_Sml_Att
 maxDistance=2000
 soundClassObject=&Fort_WeaponRanged_Fire_1P_PSM
 volumeMultiplier=0.2
 attenuationOverrides {
  bAttenuateWithLPF=true
  falloffDistance=3000
  lPFFrequencyAtMax=500
  lPFRadiusMax=1500
  lPFRadiusMin=1000
  stereoSpread=800
  attenuationShapeExtents {x=200}
 }
 concurrencyOverrides {resolutionRule=PreventNew}
 concurrencySet [&Weapon_Gunfire_FirstPerson_Conc]
}
