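# Sound cue for the PaddleGinger landing sound, set up with weapon-drop attenuation and the loot-collect sound class.
# A single landed wave plays through a Modulator that randomizes pitch slightly while keeping volume fixed.
# The cue adds a mild volume boost and joins the Weapon_Drop_StopOldest concurrency group to limit overlapping plays.
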SoundCue /MotherGameplay/Sounds/PaddleGinger/PaddleGinger_Landed_Cue
root=Modulator_0
nodes [
 {id=Modulator_0, children [WavePlayer_0], properties {pitchMin=0.97, volumeMax=1, volumeMin=1}}
 {id=WavePlayer_0, properties {soundWaveAssetPtr=&/PaddleGinger_Landed_01}}
]
properties {
 attenuationSettings=&Fort_Weapon_Drop_Att
 maxDistance=3000
 soundClassObject=&Fort_Loot_Collect
 volumeMultiplier=1.15
 concurrencyOverrides {resolutionRule=PreventNew}
 concurrencySet [&Weapon_Drop_StopOldest]
}
