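# Sound cue for selecting the PaddleGinger, routed to the weapon foley sound class and played at reduced volume.
# A single wave player feeds a modulator that varies pitch slightly on each play while holding volume steady.
# It references small-pistol attenuation settings and carries low-pass distance filter and stereo spread overrides.
# Its own concurrency override caps simultaneous instances and sets a short retrigger time, and debug is enabled.
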SoundCue /MotherGameplay/Sounds/PaddleGinger/PaddleGinger_Select_Cue
root=Modulator_0
nodes [
 {id=Modulator_0, children [WavePlayer_0], properties {pitchMax=1.03, pitchMin=0.97, volumeMax=1, volumeMin=1}}
 {id=WavePlayer_0, properties {soundWaveAssetPtr=&/PaddleGinger_Select_01}}
]
properties {
 attenuationSettings=/Game/Sounds/Attenuation/Fort_Weapon_Pistol_Sml_Att
 bDebug=true
 bOverrideConcurrency=true
 maxDistance=2000
 soundClassObject=&Fort_WeaponFoley
 volumeMultiplier=0.251
 attenuationOverrides {bAttenuateWithLPF=true, lPFFrequencyAtMax=500, lPFRadiusMax=1500, lPFRadiusMin=1000, stereoSpread=800}
 concurrencyOverrides {maxCount=2, retriggerTime=0.1}
}
