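# PaddleGinger_Transform_Cue is a sound cue for the PaddleGinger transform sound.
# A single wave player feeds PaddleGinger_Transform_01 straight to the output at reduced volume, with no randomization or modulation.
# It borrows first-person ranged-weapon fire routing: that sound class, a small-pistol attenuation asset and first-person gunfire concurrency.
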SoundCue /MotherGameplay/Sounds/PaddleGinger/PaddleGinger_Transform_Cue
root=WavePlayer_0
nodes [{id=WavePlayer_0, properties {soundWaveAssetPtr=&/PaddleGinger_Transform_01}}]
properties {
 attenuationSettings=/Game/Sounds/Attenuation/Fort_Weapon_Pistol_Sml_Att
 maxDistance=2000
 soundClassObject=&Fort_WeaponRanged_Fire_1P_PSM
 volumeMultiplier=0.7
 attenuationOverrides {distanceAlgorithm=NaturalSound, falloffDistance=4000, attenuationShapeExtents {x=1000}}
 concurrencyOverrides {resolutionRule=PreventNew}
 concurrencySet [&Weapon_Gunfire_FirstPerson_Conc]
}
